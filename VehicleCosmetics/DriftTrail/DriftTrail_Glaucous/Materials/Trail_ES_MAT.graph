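# Translucent unlit Niagara material for the Glaucous drift trail vehicle cosmetic, with scalar switches for main or glow tiling and trail or wheel masking.
# Flow-warped panning noise, blended with caustics, distorts UVs into a soft band, filled by multiply/add-blended caustics.
# Particle alpha and circular or wheel-wrap masks shape opacity; hue-shifted, desaturated particle color drives emissive.
# Static switches enable layered sine-wave vertex offset and mask-driven distortion.
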Material /VehicleCosmetics/DriftTrail/DriftTrail_Glaucous/Materials/Trail_ES_MAT
graph {
 properties {
  bAutomaticallySetUsageInEditor=false
  bUsedWithNiagaraMeshParticles=true
  bUsedWithNiagaraRibbons=true
  bUsedWithNiagaraSprites=true
  blendMode=BLEND_Translucent
  refractionMethod=RM_2DOffset
  shadingModel=MSM_Unlit
  usedShadingModels=Unlit
 }
 nodes [
  MaterialExpressionVertexColor_0 = ParticleColor()
  n1 = Constant() {r=0.5}
  n2 = Constant() {r=0.5}
  n3 = Panner(Coordinate=Multiply(A=TextureCoordinate() {vTiling=0.125}, B=Constant2Vector() {g=1, r=6}), Time=@n2) {speedY=0.5}
  n4 = Constant() {r=3}
  n5 = TextureCoordinate() {uTiling=0.25, vTiling=0.25}
  n6 = TextureSample(UVs=Panner(Coordinate=Multiply(A=@n4, B=@n5)) {speedX=-0.125, speedY=0.01}) {texture=/DealershipSharedLibrary/FX/Textures/Noise/T_VCVFX_Flow02_N}
  n7 = Multiply(A=Constant() {r=0.05}, B=AppendVector(A=@n6.R, B=@n6.R))
  n8 = Multiply(A=Time(), B=Constant() {r=3})
  n9 = TextureSample(UVs=Panner(Coordinate=Add(A=@n3, B=@n7), Time=@n8) {speedX=-1}) {texture=/DealershipSharedLibrary/FX/Textures/Noise/T_VCVFX_DistortNoise02_RGB}
  n10 = Panner(Coordinate=Multiply(A=Constant2Vector() {g=1, r=3}, B=TextureCoordinate() {uTiling=1.25, vTiling=-0.125}), Time=@n2) {speedX=0.25, speedY=-0.5}
  n11 = TextureSample(UVs=Panner(Coordinate=Add(A=@n10, B=@n7), Time=@n8) {speedX=-1}) {texture=&T_VCVFX_DistortNoise02_RGB}
  n12 = ScalarParameter() {parameterName="0=MainTrail/1=GlowTrail"}
  n13 = LinearInterpolate(A=TextureCoordinate() {uTiling=1.25, vTiling=0.5}, B=TextureCoordinate() {uTiling=0.9, vTiling=0.5}, Alpha=@n12)
  n14 = Constant() {r=0.5}
  n15 = Multiply(A=Time(), B=ScalarParameter() {defaultValue=2, parameterName=PanSpeed})
  n16 = TextureSample(UVs=Panner(Coordinate=Panner(Coordinate=@n13, Time=@n14) {speedY=0.5}, Time=@n15) {speedX=-1, speedY=0.125}) {texture=/DealershipSharedLibrary/FX/Textures/Noise/T_VCVFX_CausticsMaskSharp02_RGB}
  n17 = LinearInterpolate(A=TextureCoordinate() {vTiling=-0.5}, B=TextureCoordinate() {uTiling=0.7, vTiling=-0.5}, Alpha=@n12)
  n18 = TextureSample(UVs=Panner(Coordinate=Panner(Coordinate=@n17, Time=@n14) {speedY=-0.5}, Time=@n15) {speedX=-0.5, speedY=0.1}) {texture=&T_VCVFX_CausticsMaskSharp02_RGB}
  n19 = LinearInterpolate(A=ComponentMask(None=Add(A=@n9.RGB, B=@n11.RGB)) {g=true, r=true}, B=AppendVector(A=@n16.B, B=@n18.B), Alpha=Constant() {r=0.15})
  n20 = ComponentMask(None=Add(A=TextureCoordinate(), B=Multiply(A=@n1, B=@n19))) {g=true}
  MaterialExpressionConstantClamp_6 = Saturate(None=ConstantBiasScale(None=@n20) {bias=-0.13, scale=1})
  MaterialExpressionConstantClamp_7 = Saturate(None=ConstantBiasScale(None=OneMinus(None=@n20)) {bias=0.035, scale=1})
  n21 = Multiply(A=@MaterialExpressionConstantClamp_6, B=@MaterialExpressionConstantClamp_7)
  n22 = ScalarParameter() {defaultValue=1, parameterName=Hardness}
  MaterialExpressionConstantClamp_8 = Saturate(None=Multiply(A=@n21, B=@n22))
  n23 = ScalarParameter() {defaultValue=0.5, parameterName=SolidMask}
  n24 = Multiply(A=@MaterialExpressionConstantClamp_8, B=@n23)
  MaterialExpressionConstantClamp_5 = Saturate(None=Multiply(A=Constant() {r=20}, B=@n21))
  n25 = LinearInterpolate(A=@n24, B=@MaterialExpressionConstantClamp_5, Alpha=@n23)
  n26 = StaticComponentMaskParameter(None=@n16.RGB) {defaultG=true, parameterName=MaskRGB_01}
  n27 = StaticComponentMaskParameter(None=@n18.RGB) {defaultB=true, parameterName=MaskRGB_02}
  n28 = ScalarParameter() {defaultValue=0.5, parameterName=Brightness}
  n29 = LinearInterpolate(
   A=Multiply(A=Multiply(A=@n26, B=@n27), B=@n28)
   B=Multiply(A=@n28, B=Add(A=@n26, B=@n27))
   Alpha=ScalarParameter() {defaultValue=0.2, parameterName=AddMultiply}
  )
  MaterialExpressionConstantClamp_0 = Saturate(None=Multiply(A=@n25, B=@n29))
  MaterialExpressionConstantClamp_12 = Saturate(None=Multiply(A=@MaterialExpressionConstantClamp_0, B=@n22))
  n30 = Multiply(A=@MaterialExpressionVertexColor_0.A, B=LinearInterpolate(A=@n24, B=@MaterialExpressionConstantClamp_12, Alpha=@n23))
  n31 = Add(A=TextureCoordinate() {uTiling=2, vTiling=0.75}, B=Constant2Vector() {g=-0.385, r=-0.48})
  n32 = OneMinus(None=DotProduct(A=@n31, B=@n31))
  n33 = Constant() {r=0.785}
  MaterialExpressionConstantClamp_13 = Saturate(None=Multiply(A=Subtract(A=@n32, B=@n33), B=Constant() {r=100}))
  n34 = Add(A=TextureCoordinate() {uTiling=0.65, vTiling=0.75}, B=Constant2Vector() {g=-0.385, r=-0.48})
  n35 = OneMinus(None=DotProduct(A=@n34, B=@n34))
  n36 = Constant() {r=0.785}
  MaterialExpressionConstantClamp_14 = Saturate(None=Multiply(A=Subtract(A=@n35, B=@n36), B=Constant() {r=100}))
  MaterialExpressionConstantClamp_10 = Saturate(None=Ceil(None=Add(A=@MaterialExpressionConstantClamp_13, B=@MaterialExpressionConstantClamp_14)))
  n37 = MaterialFunctionCall!(Multiply, Power EXP=ScalarParameter() {defaultValue=0.98, parameterName=Size_FinalMask}) {materialFunction=/DealershipSharedLibrary/FX/Materials/MaterialFunctions/MF_VCVFX_WheelwrapMask}
  MaterialExpressionConstantClamp_11 = Saturate(None=LinearInterpolate(
    A=Multiply(A=@n30, B=@MaterialExpressionConstantClamp_10)
    B=Multiply(A=@n37.Result, B=@n30)
    Alpha=ScalarParameter() {parameterName="Trail=0/Wheel=1"}
   ))
  n38 = ScalarParameter() {parameterName=HueShift}
  n39 = Multiply(A=Constant() {r=0.1}, B=@n38)
  MaterialExpressionVertexColor_1 = ParticleColor()
  n40 = Add(A=@MaterialExpressionVertexColor_1.G, B=@MaterialExpressionVertexColor_1.G)
  MaterialExpressionConstantClamp_4 = Saturate(None=Add(A=LinearInterpolate(A=@n38, B=@n39, Alpha=@MaterialExpressionVertexColor_1.R), B=LinearInterpolate(A=@n39, B=@n38, Alpha=@n40)))
  MaterialExpressionConstantClamp_3 = Saturate(None=@n25)
  n41 = MaterialFunctionCall(
   Hue Shift Percentage=Multiply(
    A=StaticSwitchParameter(True=@MaterialExpressionConstantClamp_4, False=@n38) {defaultValue=true, parameterName=UseHueOffset}
    B=@MaterialExpressionConstantClamp_3
   )
   Texture=@MaterialExpressionVertexColor_0.RGB
  ) {materialFunction=&HueShift}
  n42 = ScalarParameter() {parameterName=Saturation}
  n43 = Multiply(A=Constant() {r=0.5}, B=@n42)
  MaterialExpressionVertexColor_2 = ParticleColor()
  n44 = Add(A=@MaterialExpressionVertexColor_2.G, B=@MaterialExpressionVertexColor_2.G)
  MaterialExpressionConstantClamp_9 = Saturate(None=Add(A=LinearInterpolate(A=@n42, B=@n43, Alpha=@MaterialExpressionVertexColor_2.R), B=LinearInterpolate(A=@n43, B=@n42, Alpha=@n44)))
  n45 = ComponentMask(None=TextureCoordinate()) {r=true}
  MaterialExpressionConstantClamp_2 = Saturate(None=OneMinus(None=@n45))
  n46 = LinearInterpolate(
   A=Desaturation(
    None=@n41.Result
    Fraction=StaticSwitchParameter(True=@MaterialExpressionConstantClamp_9, False=@n42) {defaultValue=true, parameterName=UseSaturationOffset}
   )
   B=@n41.Result
   Alpha=@MaterialExpressionConstantClamp_2
  )
  MaterialExpressionConstantClamp_1 = Saturate(None=@n25)
  MaterialFunctionCall(Emissive=Multiply(A=ScalarParameter() {defaultValue=2.5, parameterName=FinalEmissiveMulti}, B=Multiply(A=@n46, B=@MaterialExpressionConstantClamp_1))) {materialFunction=&MF_FX_EmissiveAdjustments}
  Reroute(None=@n6.RGB)
  MaterialExpressionConstantClamp_15 = Saturate(None=ConstantBiasScale(None=@n45) {bias=-0.01, scale=8})
  n47 = Multiply(A=ScalarParameter() {group=WPO, parameterName=Speed}, B=OneMinus(None=Time()))
  n48 = ComponentMask(None=Multiply(A=TextureCoordinate() {coordinateIndex=1, uTiling=3}, B=ScalarParameter() {group=WPO, parameterName=tile})) {r=true}
  n49 = Add(A=@n47, B=@n48)
  n50 = Multiply(A=@n47, B=Constant() {r=0.5})
  n51 = Add(A=Add(A=Cosine(None=@n49), B=Cosine(None=@n49) {period=2}), B=Cosine(None=Add(A=@n48, B=@n50)) {period=0.75})
  n52 = Add(A=@n48, B=@n47)
  n53 = Sine(None=@n52) {period=2}
  n54 = Sine(None=@n52) {period=0.75}
  n55 = AppendVector(A=@n51, B=Add(A=Add(A=@n53, B=@n54), B=Sine(None=@n52)))
  n56 = Multiply(
   A=Multiply(A=AppendVector(A=@n55, B=@n51), B=ScalarParameter() {group=WPO, parameterName=Amount})
   B=ScalarParameter() {group=WPO, parameterName=Intensity}
  )
  StaticSwitchParameter(True=Multiply(A=@MaterialExpressionConstantClamp_15, B=Transform(None=@n56)), False=Constant()) {group=WPO, parameterName=WPO}
  n57 = Constant() {r=30}
  n58 = Multiply(A=@n21, B=@n29)
  StaticSwitchParameter(
   True=Multiply(A=Multiply(A=@n58, B=@n57), B=Multiply(A=@MaterialExpressionConstantClamp_15, B=@MaterialExpressionConstantClamp_2))
   False=Constant()
  ) {parameterName=Distortion}
 ]
}
