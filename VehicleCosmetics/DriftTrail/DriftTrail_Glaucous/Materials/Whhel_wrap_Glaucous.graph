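# Translucent Niagara material for the Glaucous drift trail cosmetic, apparently a wheel wrap effect by its name, usable on sprites, ribbons and mesh particles.
# Particle color passes through a multiply into the shared FX emissive adjustments function.
# A separate branch pans the sierra_v4 texture along U and multiplies its blue channel by a constant of one.
# The graph shows no links from either branch to the material outputs.
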Material /VehicleCosmetics/DriftTrail/DriftTrail_Glaucous/Materials/Whhel_wrap_Glaucous
graph {
 properties {
  bAutomaticallySetUsageInEditor=false
  bUsedWithNiagaraMeshParticles=true
  bUsedWithNiagaraRibbons=true
  bUsedWithNiagaraSprites=true
  blendMode=BLEND_Translucent
  usedShadingModels=Default Lit
 }
 nodes [
  MaterialExpressionVertexColor_0 = ParticleColor()
  MaterialFunctionCall(Emissive=Multiply(A=@MaterialExpressionVertexColor_0.RGB)) {materialFunction=&MF_FX_EmissiveAdjustments}
  n1 = TextureCoordinate()
  Multiply(
   A=Constant() {r=1}
   B=TextureSample(UVs=Panner(Coordinate=@n1) {speedX=4}) {texture=/VehicleCosmetics/DriftTrail/DriftTrail_Glaucous/Textures/sierra_v4}.B
  )
 ]
}
